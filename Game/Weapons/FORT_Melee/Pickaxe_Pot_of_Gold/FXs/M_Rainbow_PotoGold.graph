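# Translucent unlit two-sided rainbow material for mesh particles in the Pot of Gold pickaxe effects.
# A rainbow gradient, offset by ColorScroll and mirrored in U by view-facing, is scaled by time-of-day brightness for emissive.
# Opacity combines sine edge fades, depth fade (skipped on mobile) and particle alpha.
# A camera-to-pivot cross tangent-basis vector spreads vertices sideways into a camera-facing strip tapering from base to tip width.
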Material /Game/Weapons/FORT_Melee/Pickaxe_Pot_of_Gold/FXs/M_Rainbow_PotoGold
graph {
 properties {
  bEnableResponsiveAA=true
  bUsedWithMeshParticles=true
  bUsedWithNiagaraMeshParticles=true
  blendMode=BLEND_Translucent
  pixelDepthOffsetMode=PDOM_Legacy
  shadingModel=MSM_Unlit
  translucencyLightingMode=TLM_VolumetricPerVertexNonDirectional
  twoSided=true
  usedShadingModels=Unlit
 }
 nodes [
  n1 = ScalarParameter() {defaultValue=1, parameterName=Brightness_Morning}
  n2 = ScalarParameter() {defaultValue=1.4, parameterName=Brightness_Day}
  n3 = ScalarParameter() {defaultValue=1, parameterName=Brightness_Evening}
  n4 = ScalarParameter() {defaultValue=0.65, parameterName=Brightness_Night}
  n5 = TextureCoordinate()
  n6 = MaterialFunctionCall(Float2=@n5) {materialFunction=&BreakOutFloat2Components}
  n7 = OneMinus(None=@n6.R)
  n8 = Clamp(None=DotProduct(A=VertexNormalWS(), B=CameraVectorWS()))
  n9 = Add(
   A=LinearInterpolate(
    A=@n5
    B=AppendVector(A=@n7, B=@n6.G)
    Alpha=MaterialFunctionCall!(Mask Offset Value, Number Before White Result, Gradient=@n8) {materialFunction=/Engine/Functions/Engine_MaterialFunctions02/ValueStep}.Result
   )
   B=ScalarParameter() {defaultValue=0.2, parameterName=ColorScroll}
  )
  MaterialFunctionCall(Emissive=Multiply(
    A=MaterialFunctionCall(Color_Morning=@n1, Color_Day=@n2, Color_Evening=@n3, Color_Night=@n4) {materialFunction=&Color_Time_of_Day}.Result
    B=TextureSample(UVs=@n9) {texture=/Game/Global/Textures/Default/Masks_and_Gradients/T_Grad_Rainbow}.RGB
   )) {materialFunction=&MF_FX_EmissiveAdjustments}
  n10 = TextureCoordinate()
  n11 = ComponentMask(None=@n10) {r=true}
  n12 = ComponentMask(None=@n10) {g=true}
  n13 = Multiply(
   A=Power(Base=Sine(None=@n11) {period=2}, Exp=ScalarParameter() {defaultValue=2, parameterName=U_Fade})
   B=Power(Base=Sine(None=@n12) {period=2}, Exp=ScalarParameter() {defaultValue=1, parameterName=V_Fade})
  )
  n14 = DepthFade(Opacity=@n13, FadeDistance=ScalarParameter() {defaultValue=50, parameterName=Fade_Distance})
  Multiply(
   A=Multiply(
    A=ScalarParameter() {defaultValue=0.7, parameterName=Overall_Opacity}
    B=ShadingPathSwitch(Default=@n14, Mobile=@n13) {desc=Migrated from FeatureLevelSwitch}
   )
   B=ParticleColor().A
  )
  n15 = Subtract(
   A=CameraPositionWS()
   B=MaterialFunctionCall() {materialFunction=&ComponentPivotLocation}.Result
  )
  n16 = MaterialFunctionCall() {materialFunction=&TangentBasis}
  n17 = ScalarParameter() {defaultValue=25, parameterName=Width_Base}
  n18 = ScalarParameter() {defaultValue=1.5, parameterName=Width_Tip}
  n19 = TextureCoordinate()
  n20 = ComponentMask(None=@n19) {g=true}
  n21 = Add(A=Constant() {r=-0.5}, B=@n19)
  Multiply(
   A=Normalize(VectorInput=CrossProduct(A=@n15, B=@n16.Y V))
   B=Multiply(A=LinearInterpolate(A=@n17, B=@n18, Alpha=@n20), B=ComponentMask(None=@n21) {r=true})
  )
  WorldPosition()
 ]
}
